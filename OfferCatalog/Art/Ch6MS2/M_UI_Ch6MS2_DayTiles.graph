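# Translucent UI material for Ch6MS2 day tiles in the offer catalog, showing animated falling donuts over a purple gradient.
# A linear gradient blends three purple shades as the backdrop.
# Screen-space UVs pass through a time-rotated perspective warp, then feed two panning samples of one donut texture at different scales and speeds for depth.
# Texture alphas composite the layers over the gradient, and the Item Shop rounded-corner function supplies the tile shape.
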Material /OfferCatalog/Art/Ch6MS2/M_UI_Ch6MS2_DayTiles
graph {
 properties {
  bAutomaticallySetUsageInEditor=false
  bUsedWithStaticMesh=false
  blendMode=BLEND_Translucent
  materialDomain=MD_UI
  usedShadingModels=Default Lit
 }
 nodes [
  n1 = Constant3Vector() {constant=<rgba 0.7835 0.3419 0.8469 1>}
  n2 = MaterialFunctionCall() {materialFunction=&MF_UI_UVScaling}
  n3 = Constant()
  n4 = Constant() {r=1}
  n5 = Constant() {r=0.25}
  n6 = MaterialFunctionCall(
   1=@n1
   2=Constant3Vector() {constant=<rgba 0.521 0.227 0.5647 1>}
   3=Constant3Vector() {constant=<rgba 0.2831 0.1301 0.3185 1>}
   A=OneMinus(None=MaterialFunctionCall(UVs=@n2.ScaleY, GradientStart=@n3, GradientEnd=@n4, RotationAngle=@n5) {materialFunction=&MF_UI_LinearGradient}.Result)
  ) {materialFunction=&Lerp_Multiple_Float3}
  n7 = MaterialFunctionCall!(
   UV=MaterialFunctionCall!(ScreenUV, TextureSize, ScalePivotPoint) {materialFunction=/FortUILibrary/MaterialFunctions/Basics/MF_UI_ScreenSpaceUVScaling}.ScaleY
   Perspective Rotation=Multiply(A=Time()) {constB=0.05}
   Perspective Amount=Constant() {r=0.1}
  ) {materialFunction=/FortUILibrary/MaterialFunctions/Basics/MF_UI_Perspective}
  n8 = Constant() {r=0.75}
  n9 = Multiply(A=Time()) {constB=0.45}
  n10 = TextureSample(UVs=Panner(
    Coordinate=MaterialFunctionCall(UVs=@n7.Perspective UVs, Texture Scale=@n8) {materialFunction=&ScaleUVsByCenter}.UVs
    Time=@n9
   ) {speedY=-0.05}) {texture=&/T_DonutsFalling}
  n11 = LinearInterpolate(A=@n6.Lerp 3 Inputs, B=@n10.RGB, Alpha=Multiply(A=@n10.A) {constB=0.3})
  n12 = TextureSample(UVs=Panner(Coordinate=@n7.Perspective UVs, Time=@n9) {speedX=0.01, speedY=-0.1}) {texture=&T_DonutsFalling}
  n13 = Multiply(A=@n12.A, B=@n1)
  LinearInterpolate(
   A=@n6.Lerp 3 Inputs
   B=LinearInterpolate(A=LinearInterpolate(A=@n11, B=@n12.RGB, Alpha=@n12.RGB), B=LinearInterpolate(A=@n13, B=@n12.RGB), Alpha=@n12.A)
   Alpha=Max(A=@n10.A, B=@n12.A)
  )
  MaterialFunctionCall!(Input Alpha=Constant() {r=1}, bOverrideCornerRadius, Corner Radius Override) {materialFunction=/FortUILibrary/Systems/ItemShopTile/Materials/MaterialFunctions/MF_UI_ItemShop_RoundedCorners}
 ]
}
